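# Translucent, volumetric-lit material for the idle FX on the WingBath pickaxe, usable on Niagara mesh particles and skeletal meshes.
# Color adds contrast-boosted noise highlights, panned on UV1 with a dynamic-parameter time offset, to a diffuse texture tinted by particle color.
# Opacity combines an opacity texture and particle alpha with depth fade, view-angle fade and masking.
# Named reroutes provide UVs subtly distorted by scrolling cell noise and a bottom-edge mask.
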Material /BRCosmetics/Weapons/FORT_Melee/Pickaxe_WingBath/FX/M_FX_Pickaxe_WingBath_Idle
graph {
 properties {
  bAutomaticallySetUsageInEditor=false
  bUsedWithMeshDeformer=true
  bUsedWithNiagaraMeshParticles=true
  bUsedWithSkeletalMesh=true
  blendMode=BLEND_Translucent
  translucencyLightingMode=TLM_VolumetricDirectional
  usedShadingModels=Default Lit
 }
 nodes [
  n1 = TextureCoordinate() {coordinateIndex=1}
  n2 = VectorParameter() {parameterName=NoiseScale, defaultValue=<rgba 0 5 6 4>}
  n3 = ComponentMask(None=@n2.RGB) {g=true, r=true}
  n4 = Time()
  n5 = DynamicParameter() {paramNames [Param1,Param2,IsFrontEnd,Param4]}
  n6 = TextureSample(UVs=Panner(Coordinate=Multiply(A=@n1, B=@n3), Time=Add(A=@n4, B=@n5.Param2), Speed=AppendVector(A=@n2.B, B=@n2.B)) {speedY=-3}) {samplerType=SAMPLERTYPE_Grayscale, texture=&LowResBlurredNoise}
  n7 = ScalarParameter() {defaultValue=5, parameterName=HighlightsCtrst}
  n8 = NamedRerouteUsage()
  Add(
   A=Multiply(A=Power(Base=@n6.B, Exp=@n7), B=ScalarParameter() {defaultValue=0.3, parameterName=HighlightsStr})
   B=Multiply(A=TextureSample(UVs=@n8) {texture=&/T_FX_Pickaxe_WingBath_Diff}.RGB, B=ParticleColor().RGB)
  )
  n9 = TextureSample(UVs=NamedRerouteUsage()) {samplerType=SAMPLERTYPE_LinearColor, texture=&/T_FX_Pickaxe_WingBath_Opacity}
  n10 = Constant() {r=1}
  n11 = Multiply(A=Multiply(A=Power(Base=@n9.R, Exp=@n10), B=Constant() {r=1}), B=ParticleColor().A)
  n12 = ScalarParameter() {parameterName=FadeDistance}
  n13 = Constant() {r=0.8}
  Multiply(
   A=Multiply(
    A=DepthFade(Opacity=@n11, FadeDistance=@n12)
    B=MaterialFunctionCall("FadeSensitivity (optional)"=@n13) {materialFunction=&FadeBasedOnViewAngle}.Result
   )
   B=NamedRerouteUsage()
  )
  n14 = Panner(Coordinate=TextureCoordinate()) {speedY=1}
  n15 = Constant2Vector() {g=1}
  n16 = LinearInterpolate(Alpha=TextureSample(UVs=Multiply(A=@n14, B=@n15)) {texture=/Game/Creative/PostProcess/Textures/Cell_Noise_T1}.R) {constA=-1}
  n17 = Constant() {r=0.001}
  NamedRerouteDeclaration(None=Add(A=TextureCoordinate(), B=Multiply(A=@n16, B=@n17))) {name=UV, nodeColor=<rgba 0 1 0.7891 1>}
  n18 = OneMinus(None=MaterialFunctionCall(Float2=TextureCoordinate()) {materialFunction=&BreakOutFloat2Components}.G)
  n19 = Constant() {r=20}
  NamedRerouteDeclaration(None=Saturate(None=Multiply(A=@n18, B=@n19))) {name=BottomMask, nodeColor=<rgba 0 1 0.5078 1>}
 ]
}
