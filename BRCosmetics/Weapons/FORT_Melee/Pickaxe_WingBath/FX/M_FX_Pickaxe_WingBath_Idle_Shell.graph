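# Translucent lit material for the WingBath pickaxe idle FX shell, usable on Niagara mesh particles, with diffuse and baked normal textures.
# Particle color tints the diffuse, with added panning noise highlights on UV1, time-offset by a dynamic parameter.
# Particle alpha scales a mask, and a UV1 gradient defines a bottom mask.
# A previous-frame-only particle-space offset fakes directional motion blur through translucent velocity, zeroed on Low, Medium and High quality.
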Material /BRCosmetics/Weapons/FORT_Melee/Pickaxe_WingBath/FX/M_FX_Pickaxe_WingBath_Idle_Shell
graph {
 properties {
  bOutputTranslucentVelocity=true
  bUsedWithMeshDeformer=true
  bUsedWithNiagaraMeshParticles=true
  bUsedWithSkeletalMesh=true
  blendMode=BLEND_Translucent
  pixelDepthOffsetMode=PDOM_Legacy
  translucencyLightingMode=TLM_VolumetricPerVertexDirectional
  usedShadingModels=Default Lit
 }
 nodes [
  n1 = NamedRerouteUsage()
  n2 = TextureCoordinate() {coordinateIndex=1}
  n3 = VectorParameter() {parameterName=NoiseScale, defaultValue=<rgba 0 8 6 4>}
  n4 = ComponentMask(None=@n3.RGB) {g=true, r=true}
  n5 = Time()
  n6 = DynamicParameter() {paramNames [Param1,Param2,IsFrontEnd,Param4]}
  n7 = TextureSample(UVs=Panner(Coordinate=Multiply(A=@n2, B=@n4), Time=Add(A=@n5, B=@n6.Param2), Speed=AppendVector(A=@n3.B, B=@n3.B)) {speedY=-3}) {samplerType=SAMPLERTYPE_Grayscale, texture=&LowResBlurredNoise}
  n8 = ScalarParameter() {defaultValue=6, parameterName=HighlightsCtrst}
  Add(
   A=Multiply(A=TextureSample(UVs=@n1) {texture=/BRCosmetics/Weapons/FORT_Melee/Pickaxe_WingBath/Textures/T_WingBath_Pickaxe_D}.RGB, B=ParticleColor().RGB)
   B=Multiply(A=Power(Base=@n7.B, Exp=@n8), B=ScalarParameter() {defaultValue=0.2, parameterName=HighlightsStr})
  )
  Multiply(A=ParticleColor().A, B=NamedRerouteUsage())
  NamedRerouteDeclaration(None=TextureCoordinate()) {name=UV, nodeColor=<rgba 0 1 0.7891 1>}
  n9 = OneMinus(None=MaterialFunctionCall(Float2=TextureCoordinate() {coordinateIndex=1}) {materialFunction=&BreakOutFloat2Components}.G)
  n10 = Constant() {r=10}
  NamedRerouteDeclaration(None=Saturate(None=Multiply(A=@n9, B=@n10))) {name=BottomMask, nodeColor=<rgba 0 1 0.5078 1>}
  n11 = PreviousFrameSwitch(
   Current Frame=Constant3Vector()
   Previous Frame=Transform(None=VectorParameter() {parameterName=BlurDirectionAndStrength, defaultValue=<rgba 40 0 0 0>}.RGB) {transformSourceType=TRANSFORMSOURCE_ParticleWorld}
  )
  n12 = Constant3Vector()
  QualitySwitch(Default=@n11, Low=@n12, High=@n12, Medium=@n12, Epic=@n11)
  TextureSampleParameter2D(UVs=NamedRerouteUsage()) {
   group=Mask Texture
   parameterName=BakedNormal
   samplerSource=SSM_Wrap_WorldGroupSettings
   samplerType=SAMPLERTYPE_Normal
   texture=/BRCosmetics/Weapons/FORT_Melee/Pickaxe_WingBath/Textures/T_WingBath_Pickaxe_N
  }
 ]
}
