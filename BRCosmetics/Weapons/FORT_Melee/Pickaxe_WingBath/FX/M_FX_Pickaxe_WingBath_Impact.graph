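# Additive unlit impact effect material for the WingBath pickaxe, used on Niagara sprites.
# The impact texture's blue channel is refined by MF_FX_Refine with a vector parameter, then multiplied by particle color and alpha.
# The result feeds MF_FX_EmissiveAdjustments.
# The ChannerGOrB parameter blends blue with itself, so it has no effect.
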Material /BRCosmetics/Weapons/FORT_Melee/Pickaxe_WingBath/FX/M_FX_Pickaxe_WingBath_Impact
graph {
 properties {
  bAutomaticallySetUsageInEditor=false
  bUsedWithNiagaraSprites=true
  blendMode=BLEND_Additive
  shadingModel=MSM_Unlit
  usedShadingModels=Unlit
 }
 nodes [
  n1 = ParticleColor()
  n2 = TextureSample() {samplerType=SAMPLERTYPE_LinearColor, texture=&/T_FX_Pickaxe_WingBath_Impact}
  n3 = VectorParameter() {defaultValue=<rgba 5 5 0.5 0.5>}
  n4 = MaterialFunctionCall(Texture=LinearInterpolate(A=@n2.B, B=@n2.B, Alpha=ScalarParameter() {parameterName=ChannerGOrB}), RGB=@n3.RGB, A=@n3.RGB) {materialFunction=&MF_FX_Refine}
  MaterialFunctionCall(Emissive=Multiply(A=Multiply(A=@n1.RGB, B=@n4.Result), B=@n1.A)) {materialFunction=&MF_FX_EmissiveAdjustments}
 ]
}
