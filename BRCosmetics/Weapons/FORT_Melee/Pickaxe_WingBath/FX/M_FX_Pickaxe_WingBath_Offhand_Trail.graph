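# Additive unlit Niagara ribbon material for the trail of the WingBath pickaxe's offhand weapon.
# Sums tinted layers: a soft base and inner line from refined sphere masks, side and soft lines from channels of one packed texture, and sine-generated vertical lines.
# UV-based length, width and outer masks are shared through named reroutes.
# The result feeds an emissive-adjustment function, and particle alpha drives depth-faded opacity.
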Material /BRCosmetics/Weapons/FORT_Melee/Pickaxe_WingBath/FX/M_FX_Pickaxe_WingBath_Offhand_Trail
graph {
 properties {
  bAutomaticallySetUsageInEditor=false
  bUsedWithNiagaraRibbons=true
  blendMode=BLEND_Additive
  shadingModel=MSM_Unlit
  usedShadingModels=Unlit
 }
 nodes [
  DepthFade(Opacity=ParticleColor().A, FadeDistance=ScalarParameter() {defaultValue=10, parameterName=FadeDistance})
  n1 = NamedRerouteUsage()
  n2 = Multiply(A=NamedRerouteUsage(), B=NamedRerouteUsage())
  n3 = Add(A=Add(A=Add(A=@n1, B=@n2), B=NamedRerouteUsage()), B=NamedRerouteUsage())
  n4 = NamedRerouteUsage()
  MaterialFunctionCall(Emissive=Multiply(A=Add(A=@n3, B=@n4), B=NamedRerouteUsage())) {materialFunction=&MF_FX_EmissiveAdjustments}
  n5 = MaterialFunctionCall(Float2=TextureCoordinate()) {materialFunction=&BreakOutFloat2Components}
  n6 = Multiply(A=@n5.R, B=Constant() {r=5})
  NamedRerouteDeclaration(None=Multiply(A=OneMinus(None=@n5.R), B=Saturate(None=@n6))) {name=LengthMask, nodeColor=<rgba 1 0.5859 0 1>}
  n7 = ScalarParameter() {parameterName=Param_2}
  n8 = Add(A=@n7, B=ScalarParameter() {parameterName=Param_3})
  n9 = Constant() {r=0.5}
  n10 = SphereMask(
   A=MaterialFunctionCall(Float2=TextureCoordinate()) {materialFunction=&BreakOutFloat2Components}.G
   B=@n9
   Radius=@n9
  ) {hardnessPercent=0}
  NamedRerouteDeclaration(None=Saturate(None=SmoothStep(Min=@n7, Max=@n8, Value=@n10))) {name=WidthMask, nodeColor=<rgba 1 0 0.1406 1>}
  n11 = Constant() {r=0.5}
  n12 = SphereMask(
   A=MaterialFunctionCall(Float2=TextureCoordinate()) {materialFunction=&BreakOutFloat2Components}.G
   B=@n11
   Radius=@n11
  ) {hardnessPercent=0}
  n13 = Power(Base=OneMinus(None=@n12), Exp=Constant() {r=2})
  n14 = VectorParameter() {parameterName=Refine, defaultValue=<rgba 3 5 1 0.5>}
  n15 = ScalarParameter() {defaultValue=0.11}
  n16 = ScalarParameter() {parameterName=Param_1}
  n17 = Saturate(None=SmoothStep(Min=@n15, Max=Add(A=@n15, B=@n16), Value=@n12))
  n18 = Multiply(
   A=Saturate(None=MaterialFunctionCall(Texture=@n13, RGB=@n14.RGB, A=@n14.RGB) {materialFunction=&MF_FX_Refine}.Result)
   B=@n17
  )
  n19 = VectorParameter() {parameterName=Color, defaultValue=<rgba 0.1417 0.3778 0.7083 1>}
  NamedRerouteDeclaration(None=Multiply(A=Multiply(A=@n18, B=@n19.RGB), B=ScalarParameter() {defaultValue=0.5, parameterName=BaseMultiplier})) {name=Base, nodeColor=<rgba 1 0 0.8203 1>}
  n20 = VectorParameter() {parameterName=Param_4, defaultValue=<rgba 1 1 0 0>}
  n21 = TextureObject() {texture=&/T_FX_Pickaxe_WingBath_Offhand_Trail}
  n22 = TextureSample(
   UVs=MaterialFunctionCall(
    UVs=TextureCoordinate()
    Texture Scale=MaterialFunctionCall(X=@n20.R, Y=@n20.R) {materialFunction=&MakeFloat2}.Result
   ) {materialFunction=&ScaleUVsByCenter}.UVs
   Tex=@n21
  ) {texture=&T_FX_Pickaxe_WingBath_Offhand_Trail}
  n23 = Add(A=@n22.R, B=Multiply(A=@n22.B, B=ScalarParameter() {defaultValue=0.05, parameterName=Param_5})) {constB=0}
  n24 = VectorParameter() {parameterName=Color2, defaultValue=<rgba 0.6 1.7 5 1>}
  NamedRerouteDeclaration(None=Multiply(A=Multiply(A=@n23, B=@n24.RGB), B=ScalarParameter() {defaultValue=1, parameterName=SideLinesMultiplier})) {name=SideLines, nodeColor=<rgba 1 0.75 0 1>}
  n25 = TextureSample(Tex=@n21) {texture=&T_FX_Pickaxe_WingBath_Offhand_Trail}
  n26 = Multiply(A=@n25.G, B=VectorParameter() {parameterName=Param_1, defaultValue=<rgba 0.0778 0.3024 0.6 1>}.RGB)
  n27 = MaterialFunctionCall(Float2=TextureCoordinate()) {materialFunction=&BreakOutFloat2Components}
  n28 = ScalarParameter() {defaultValue=3, parameterName=Param_6}
  n29 = Multiply(A=OneMinus(None=@n27.R), B=Saturate(None=Multiply(A=@n27.R, B=@n28)))
  NamedRerouteDeclaration(None=Multiply(A=Multiply(A=@n26, B=@n29), B=ScalarParameter() {defaultValue=1, parameterName=SoftLinesMultiplier})) {name=SoftLines, nodeColor=<rgba 0.8516 1 0 1>}
  n30 = TextureCoordinate()
  n31 = VectorParameter() {defaultValue=<rgba 5 0.5 1 0.3>}
  n32 = MaterialFunctionCall(
   Texture=SphereMask(
    A=MaterialFunctionCall(Float2=@n30) {materialFunction=&BreakOutFloat2Components}.G
    B=Constant() {r=0.5}
    Radius=ScalarParameter() {defaultValue=0.06, parameterName=Param_4}
   ) {hardnessPercent=0}
   RGB=@n31.RGB
   A=@n31.RGB
  ) {materialFunction=&MF_FX_Refine}
  n33 = VectorParameter() {parameterName=Param_2, defaultValue=<rgba 0.4 0.65 1 1>}
  NamedRerouteDeclaration(None=Multiply(A=Multiply(A=@n32.Result, B=@n33.RGB), B=ScalarParameter() {defaultValue=0.6, parameterName=InnerLineMultiplier})) {name=InnerLine, nodeColor=<rgba 0 0.8125 1 1>}
  n34 = MaterialFunctionCall(Float2=TextureCoordinate()) {materialFunction=&BreakOutFloat2Components}
  n35 = ScalarParameter() {defaultValue=12, parameterName=Sine}
  n36 = ConstantBiasScale(None=Sine(None=Multiply(A=@n34.R, B=@n35)))
  n37 = ScalarParameter() {defaultValue=80, parameterName=Param_8}
  n38 = Multiply(A=Saturate(None=Power(Base=@n36, Exp=@n37)), B=NamedRerouteUsage())
  n39 = VectorParameter() {parameterName=Param_3, defaultValue=<rgba 0.0301 0.0601 0.12 1>}
  NamedRerouteDeclaration(None=Multiply(A=Multiply(A=@n38, B=@n39.RGB), B=ScalarParameter() {defaultValue=1, parameterName=VerticalLinesMultiplier})) {name=VerticalLines, nodeColor=<rgba 1 0 0.3516 1>}
  NamedRerouteDeclaration(None=@n17) {name=OuterMask, nodeColor=<rgba 0 0.9297 1 1>}
  NamedRerouteUsage()
  OneMinus(None=@n12)
  Reroute(None=@n25.B)
 ]
}
