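# Additive, unlit, two-sided Niagara mesh particle material for a ring glow effect on the WingBath pickaxe.
# It builds a procedural soft band with no texture sample, using a SphereMask on the U coordinate centered at 0.5, sharpened by a Power with a scalar exponent.
# MF_FX_Refine shapes the band using the Refine vector parameter, then the result is multiplied by particle color and fed into MF_FX_EmissiveAdjustments.
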Material /BRCosmetics/Weapons/FORT_Melee/Pickaxe_WingBath/FX/M_FX_WingBath_RingGlow
graph {
 properties {
  bAutomaticallySetUsageInEditor=false
  bUsedWithNiagaraMeshParticles=true
  blendMode=BLEND_Additive
  shadingModel=MSM_Unlit
  twoSided=true
  usedShadingModels=Unlit
 }
 nodes [
  n1 = TextureCoordinate()
  n2 = Constant() {r=0.5}
  n3 = Power(
   Base=SphereMask(
    A=MaterialFunctionCall(Float2=@n1) {materialFunction=&BreakOutFloat2Components}.R
    B=@n2
    Radius=@n2
   ) {hardnessPercent=0}
   Exp=ScalarParameter() {defaultValue=2}
  )
  n4 = VectorParameter() {parameterName=Refine, defaultValue=<rgba 5 1 1 1>}
  MaterialFunctionCall(Emissive=Multiply(
    A=MaterialFunctionCall(Texture=@n3, RGB=@n4.RGB, A=@n4.RGB) {materialFunction=&MF_FX_Refine}.Result
    B=ParticleColor().RGB
   )) {materialFunction=&MF_FX_EmissiveAdjustments}
 ]
}
