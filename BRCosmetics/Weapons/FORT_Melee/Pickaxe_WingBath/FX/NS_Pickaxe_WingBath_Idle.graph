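# Idle Niagara effect for the WingBath pickaxe, drawing burst-spawned translucent mesh copies that oscillate along X.
# Main spawns three copies and Minimal004 four, with per-spawn-group wave and alpha; only Minimal004's motion scales with System.ShakingIntensity.
# Time-of-day colors tint the meshes, while front-end modifiers adjust scale, opacity, tint and blur direction, with Main's opacity also platform-dependent.
# The FN_Debug sprite emitter is disabled.
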NiagaraSystem /BRCosmetics/Weapons/FORT_Melee/Pickaxe_WingBath/FX/NS_Pickaxe_WingBath_Idle
dependencies {
 usedDataInterfaces [&NiagaraDataInterfaceColorCurve, &NiagaraDataInterfacePlatformSet]
 usedDynamicInputs []
 usedModules [
  &SystemState
  &self:SystemSpawnScript.SystemScriptSource.SystemScriptGraph.NiagaraNodeAssignment_0.SetVariables_#1
 ]
}
emitters {
 FN_Debug {
  bEnabled=false
  emitterUpdateScript [
   &EmitterState()
   &SpawnBurst_Instantaneous(Spawn Count=1, Use Spawn Probability=false)
  ]
  particleSpawnScript [/Niagara/Modules/Spawn/Initialization/InitializeParticle(
    Lifetime=5
    Module.Write Color=true
    Module.Write Lifetime=true
    Module.Write Mass=true
    Module.Write Material Random=true
    Module.Write Position=false
    Module.Write Scale=false
    Module.Write SpriteRotation=false
    Module.Write SpriteSize=true
    Position=&SimulationPosition()
    Sprite Size=<100 100>
   )]
  particleUpdateScript [
   &ParticleState()
   /SetVariables(Particles.Debug:NiagaraFloat=&FN_FrontEndFloatModifier(Gameplay=0, Front End=1)) {isSetParameters=true}
   &DynamicMaterialParameters(
    Index 0 Param 1=$Particles.Debug
    Module.Param0WriteEnabled=true
    Module.Param1WriteEnabled=true
    Module.Param2WriteEnabled=true
    Module.Param3WriteEnabled=true
   )
  ]
  rendererClasses [&NiagaraSpriteRendererProperties]
  renderers [{rendererClass=&NiagaraSpriteRendererProperties}]
 }
 Main {
  emitterUpdateScript [
   &EmitterState(Life Cycle Mode=::NewEnumerator1, Loop Behavior=::NewEnumerator1, Loop Duration Mode=::NewEnumerator1, UseLoopDelay=false)
   &SpawnBurst_Instantaneous(Spawn Count=3, Use Spawn Probability=false, UseLoopCountLimit=false)
  ]
  particleSpawnScript [
   &FN_SetSpawnGroup(Number of Spawn Groups=3)
   /Niagara/Modules/Spawn/Initialization/V2/InitializeParticle(
    Color=<rgba 1 1 1 0.6>
    Lifetime=5
    Mesh Scale=&MakeVector(X=&FN_FrontEndFloatModifier(Gameplay=0.9, Front End=0.8), Y=1, Z=1)
    Mesh Scale Mode=::NewEnumerator4
    UsePositionOffset=false
    Write Material Random=true
   )
  ]
  particleUpdateScript [
   &ParticleState(Kill Particles When Lifetime Has Elapsed=false, Let Infinitely Lived Particles Die When Emitter Deactivates=true)
   /SetVariables(Particles.Position=&Lerp_Vector(
     A=Linked() {linkedVariable {name=Particles.Initial.Position, type=NiagaraPosition}}
     B=&AddVectorToPosition(
      Position=$Particles.Initial.Position
      Vector=&Multiply_VectorByFloat(
       Vector=&MakeVector(
        X=&MergeHelper_FloatSwitch(
         0=&Waveform("Waveform [1]"=::NewEnumerator0, "[1] Frequency"=12, "[1] Amplitude Min/Max"=<-1 1>, "[1] Scale"=0.3)
         1=&Waveform("Waveform [1]"=::NewEnumerator0, "[1] Frequency"=12, "[1] Amplitude Min/Max"=<-1 1>, "[1] Scale"=0.7)
         2=&Waveform("Waveform [1]"=::NewEnumerator0, "[1] Input Position"=0.5, "[1] Frequency"=10, "[1] Amplitude Min/Max"=<-1 1>, "[1] Scale"=1.3)
         3=?
         Index=$Particles.SpawnGroup
         Num=::NewEnumerator1
        )
        Y=0
        Z=0
       )
       Float=0.4
      )
     )
     Alpha=1
    )) {isSetParameters=true}
   &ScaleColor(
    Scale Alpha=&Lerp_Float(
     A=&MergeHelper_FloatSwitch(
      0=&Multiply_Float(A=1, B=&Waveform("Waveform [1]"=::NewEnumerator0, "[1] Frequency"=0.5, "[1] Amplitude Min/Max"=<0.5 1>))
      1=&Multiply_Float(A=0.5, B=&Waveform("Waveform [1]"=::NewEnumerator0, "[1] Frequency"=0.5, "[1] Amplitude Min/Max"=<0.5 1>))
      2=&Multiply_Float(A=0.35, B=&Waveform("Waveform [1]"=::NewEnumerator0, "[1] Frequency"=0.5, "[1] Amplitude Min/Max"=<0.5 1>))
      3=?
      Index=$Particles.SpawnGroup
      Num=::NewEnumerator1
     )
     B=0.5
     Alpha=$System.ShakingIntensity
    )
    ScaleA=true
    ScaleRGB=false
   ) {enabled=false}
   &CameraOffset(Camera Offset Amount=1)
   &ScaleColor(
    Scale RGB=&Lerp_Vector(
     A=&MakeVectorFromLinearColorRGB(Color=&LerpLinearColors(
       StartColor=&ColorFromCurve(
        ColorCurve=DataInterface() {propertyValues {
          AlphaCurve {keys [<at 0 1>,<at 0.125 1>,<at 0.25 1>,<at 0.375 1>,<at 0.5 1>,<at 0.625 1>,<at 0.75 1>,<at 0.875 1>,<at 1 1>]}
          BlueCurve {keys [<at 0 0.9216>,<at 0.041 0.8963>,<at 0.083 0.8308>,<at 0.125 0.7835>,<at 0.166 0.6172>,<at 0.208 0.5395>,<at 0.25 0.6172>,<at 0.291 0.6939>,<at 0.333 0.7379>,<at 0.375 0.7835>,<at 0.416 0.7991>,<at 0.458 0.8148>,<at 0.5 0.8308>,<at 0.541 0.8148>,<at 0.583 0.7991>,<at 0.625 0.7991>,<at 0.666 0.7529>,<at 0.708 0.7084>,<at 0.75 0.5711>,<at 0.791 0.4735>,<at 0.833 0.3231>,<at 0.875 0.4233>,<at 0.916 0.8388>,<at 0.958 0.9047>,<at 1 0.9216>]}
          GreenCurve {keys [<at 0 0.6308>,<at 0.041 0.5711>,<at 0.083 0.5333>,<at 0.125 0.5583>,<at 0.166 0.5333>,<at 0.208 0.5271>,<at 0.25 0.6376>,<at 0.291 0.7529>,<at 0.333 0.7835>,<at 0.375 0.8148>,<at 0.416 0.7991>,<at 0.458 0.8148>,<at 0.5 0.8469>,<at 0.541 0.7991>,<at 0.583 0.7991>,<at 0.625 0.8148>,<at 0.666 0.7991>,<at 0.708 0.7529>,<at 0.75 0.5711>,<at 0.791 0.5029>,<at 0.833 0.5029>,<at 0.875 0.6172>,<at 0.916 0.5333>,<at 0.958 0.5711>,<at 1 0.6308>]}
          RedCurve {keys [<at 0 0.305>,<at 0.041 0.2623>,<at 0.083 0.2664>,<at 0.125 0.2831>,<at 0.166 0.7084>,<at 0.208 0.6867>,<at 0.25 0.8148>,<at 0.291 0.8879>,<at 0.333 0.8963>,<at 0.375 0.9131>,<at 0.416 0.8714>,<at 0.458 0.8714>,<at 0.5 0.9131>,<at 0.541 0.8632>,<at 0.583 0.8714>,<at 0.625 0.9131>,<at 0.666 0.8963>,<at 0.708 0.8714>,<at 0.75 0.7529>,<at 0.791 0.7529>,<at 0.833 0.8469>,<at 0.875 0.8632>,<at 0.916 0.1878>,<at 0.958 0.1746>,<at 1 0.305>]}
         }}
        CurveIndex=Linked() {linkedVariable {name=NPC.FortniteNPC.TimeOfDay_0-1, type=NiagaraFloat}}
       )
       EndColor=<rgba 0.474 0.474 0.474 1>
       LerpFactor=0.5
      ))
     B=<1 1 1>
     Alpha=&FN_FrontEndFloatModifier(Gameplay=0, Front End=1)
    )
    ScaleA=false
    ScaleRGB=true
   ) {module=ScaleColor001}
   &ScaleColor(
    Scale Alpha=&MergeHelper_FloatSwitch(0=1, 1=0.3, 2=1, 3=?, Index=$Particles.SpawnGroup, Num=::NewEnumerator1)
    ScaleA=true
    ScaleRGB=false
   ) {module=ScaleColor002}
   &ScaleColor(
    Scale RGB=&MakeVectorFromLinearColorRGB(Color=&FNTimeOfDayColor(Night Color=<rgba 1 1 1 1>, Morning Color=<rgba 1 1 1 1>, Day Color=<rgba 0.77 0.9003 1 1>, Evening Color=<rgba 1 1 1 1>))
    ScaleA=false
    ScaleRGB=true
   ) {module=ScaleColor003}
   &ScaleColor(
    Scale Alpha=&FN_FrontEndFloatModifier(
     Gameplay=&FN_PickFloatDependOnPlatform(TargetPlatformActive=1, TargetPlatformNoActive=0.9)
     Front End=&FN_PickFloatDependOnPlatform(TargetPlatformActive=0.8, TargetPlatformNoActive=0.5)
    )
    ScaleA=true
    ScaleRGB=false
   ) {module=ScaleColor004}
  ]
  rendererClasses [&NiagaraMeshRendererProperties]
  renderers [{rendererClass=&NiagaraMeshRendererProperties, rendererIndex=0}]
 }
 Minimal004 {
  emitterUpdateScript [
   &EmitterState(Life Cycle Mode=::NewEnumerator1, Loop Behavior=::NewEnumerator1, Loop Duration Mode=::NewEnumerator1, UseLoopDelay=false)
   &SpawnBurst_Instantaneous(Spawn Count=4, Use Spawn Probability=false, UseLoopCountLimit=true)
   /SetVariables(Emitter.BlurDir:Vector4f=&MakeVector4(X=&FN_FrontEndFloatModifier(Gameplay=25, Front End=6), Y=0, Z=0, W=1)) {isSetParameters=true}
  ]
  particleSpawnScript [
   &InitializeParticle(
    Color=<rgba 1 1 1 0.2>
    Lifetime=5
    Mesh Scale=<1 1.1 1>
    Mesh Scale Mode=::NewEnumerator4
    UsePositionOffset=true
    Write Material Random=true
   )
   &FN_SetSpawnGroup()
  ]
  particleUpdateScript [
   &ParticleState(Kill Particles When Lifetime Has Elapsed=false, Let Infinitely Lived Particles Die When Emitter Deactivates=true)
   /SetVariables(Particles.Position=&Lerp_Vector(
     A=Linked() {linkedVariable {name=Particles.Initial.Position, type=NiagaraPosition}}
     B=&AddVectorToPosition(
      Position=$Particles.Initial.Position
      Vector=&Multiply_VectorByFloat(
       Vector=&MakeVector(
        X=&Waveform(
         "Waveform [1]"=::NewEnumerator0
         "[1] Frequency"=&Multiply_Float(
          A=&MergeHelper_FloatSwitch(0=20, 1=20, 2=20, 3=20, Index=$Particles.SpawnGroup, Num=::NewEnumerator2)
          B=1
         )
         "[1] Amplitude Min/Max"=<-1 1>
         "[1] Scale"=&Multiply_Float(
          A=&MergeHelper_FloatSwitch(0=0.3, 1=0.66, 2=0.8, 3=1, Index=$Particles.SpawnGroup, Num=::NewEnumerator2)
          B=0.5
         )
        )
        Y=0
        Z=0
       )
       Float=1
      )
     )
     Alpha=&Multiply_Float(A=$System.ShakingIntensity, B=&FN_FrontEndFloatModifier(Gameplay=1, Front End=1))
    )) {isSetParameters=true}
   &ScaleColor(
    Scale RGB=&MakeVectorFromLinearColorRGB(Color=&FNTimeOfDayColor(Night Color=<rgba 1 1 1 1>, Morning Color=<rgba 1 1 1 1>, Day Color=<rgba 0.75 0.8917 1 1>, Evening Color=<rgba 1 1 1 1>))
    ScaleA=false
    ScaleRGB=true
   ) {module=ScaleColor003}
   &ScaleColor(
    Scale Alpha=&MergeHelper_FloatSwitch(0=1, 1=1, 2=1, 3=4, Index=$Particles.SpawnGroup, Num=::NewEnumerator2)
    ScaleA=true
    ScaleRGB=false
   ) {module=ScaleColor002}
   &CameraOffset(Camera Offset Amount=0.1)
   &ScaleColor(Scale Alpha=&FN_FrontEndFloatModifier(Gameplay=1, Front End=0.5), ScaleA=true, ScaleRGB=false)
  ]
  rendererClasses [&NiagaraMeshRendererProperties]
  renderers [{rendererClass=&NiagaraMeshRendererProperties, rendererIndex=0}]
 }
}
summary {
 systemName=NS_Pickaxe_WingBath_Idle
 emitters [
  {emitterName=Main, rendererClasses [&NiagaraMeshRendererProperties]}
  {emitterName=Minimal004, rendererClasses [&NiagaraMeshRendererProperties]}
  {bEnabled=false, emitterName=FN_Debug, rendererClasses [&NiagaraSpriteRendererProperties]}
 ]
 userVariables [
  User.bIsFrontEnd {type=NiagaraBool}
  User.bIsFrontEndPreview {type=NiagaraBool}
 ]
}
