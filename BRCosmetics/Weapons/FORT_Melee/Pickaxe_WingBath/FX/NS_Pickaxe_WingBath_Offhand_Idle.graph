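# Niagara idle glow effect for the offhand WingBath pickaxe, combining meshes and a light.
# One Glow emitter bursts three persistent particles, each spawn group setting its blue-toned color and renderer visibility tag.
# Two mesh renderers and a light renderer draw them, with dynamic material parameters written at spawn.
# The blue light is zeroed in the front end through a gameplay modifier, while particle alpha scales with time of day.
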NiagaraSystem /BRCosmetics/Weapons/FORT_Melee/Pickaxe_WingBath/FX/NS_Pickaxe_WingBath_Offhand_Idle
dependencies {usedDataInterfaces [], usedDynamicInputs [], usedModules [&SystemState]}
emitters {Glow {
  emitterUpdateScript [
   &EmitterState(Life Cycle Mode=::NewEnumerator1, Loop Behavior=::NewEnumerator1, Loop Duration Mode=::NewEnumerator1, UseLoopDelay=false)
   &SpawnBurst_Instantaneous(Spawn Count=3, Use Spawn Probability=false, UseLoopCountLimit=true)
  ]
  particleSpawnScript [
   &FN_SetSpawnGroup(Number of Spawn Groups=3)
   &InitializeParticle(
    Color=&MergeHelper_ColorSwitch(
     0=<rgba 0.2344 0.8551 1.5 0.15>
     1=<rgba 0.0289 0.1065 0.1739 1>
     2=<rgba 0.237 0.4299 0.7 0.3>
     3=?
     Index=$Particles.SpawnGroup
     Num=::NewEnumerator1
    )
    Lifetime=5
    Mesh Scale Mode=::NewEnumerator3
    UsePositionOffset=true
    Write Material Random=true
   )
   /SetVariables(Particles.VisibilityTag=$Particles.SpawnGroup) {isSetParameters=true}
   &DynamicMaterialParameters(
    Index 0 Param 1=1
    Module.Index0Float0Write=true
    Module.Index0Float1Write=false
    Module.Index0Float2Write=true
    Module.Index0Float3Write=false
   )
  ]
  particleUpdateScript [
   &ParticleState(Kill Particles When Lifetime Has Elapsed=false, Let Infinitely Lived Particles Die When Emitter Deactivates=true)
   &FN_Light_Attributes(
    Color=<rgba 0.2 0.5333 1 1>
    Create Color Attribute=true
    Create Exponent Attribute=true
    Create Radius Attribute=true
    Exponent=2
    Radius=50
    Scale Color=&FN_FrontEndFloatModifier(Gameplay=5, Front End=0)
   )
   &ScaleColor(
    Scale Alpha=&FNTimeOfDayScalar(Night Scale=0.8, Morning Scale=0.8, Evening Scale=0.8)
    ScaleA=true
    ScaleRGB=true
   )
  ]
  rendererClasses [&NiagaraMeshRendererProperties, &NiagaraLightRendererProperties]
  renderers [
   {rendererClass=&NiagaraMeshRendererProperties, rendererIndex=0}
   {rendererClass=&NiagaraMeshRendererProperties, rendererIndex=1}
   {rendererClass=&NiagaraLightRendererProperties, rendererIndex=2}
  ]
 }}
summary {
 systemName=NS_Pickaxe_WingBath_Offhand_Idle
 emitters [{emitterName=Glow, rendererClasses [&NiagaraMeshRendererProperties, &NiagaraLightRendererProperties]}]
 userVariables [
  User.bIsFrontEnd {type=NiagaraBool}
  User.bIsFrontEndPreview {type=NiagaraBool}
 ]
}
