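# Material function for the Wing Bath pickaxe's offhand, building an animated translucent, shield-like energy effect and writing emissive and opacity into input material attributes.
# Radial UVs drive procedural pulsing rings, jittered streaks and an outer ring, alongside textured glow, flare and scrolling turbulence layers.
# Normal-map-distorted scene color is sampled only on higher quality levels.
# Sort-translucency opacity is adjusted using scene depth and Single Layer Water shading-model detection.
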MaterialFunction /BRCosmetics/Weapons/FORT_Melee/Pickaxe_WingBath/Materials/MF_Pickaxe_WingBath_Offhand_Parent
graph {
 properties {previewBlendMode=BLEND_Translucent}
 nodes [
  n1 = Add(A=NamedRerouteUsage(), B=NamedRerouteUsage())
  n2 = Multiply(
   A=Power(Base=NamedRerouteUsage(), Exp=ScalarParameter() {defaultValue=2, group=Wing Bath - Pickaxe - Offhand, parameterName=Turbulence - Power})
   B=ScalarParameter() {defaultValue=0.075, group=Wing Bath - Pickaxe - Offhand, parameterName=Turbulence - Intensity}
  )
  Add(A=@n1, B=@n2)
  n3 = NamedRerouteUsage()
  n4 = Time()
  n5 = NamedRerouteUsage()
  n6 = Add(
   A=Multiply(A=@n3, B=ScalarParameter() {defaultValue=2, parameterName=Secondary - Frequency}) {constB=3}
   B=Multiply(A=Multiply(A=@n4, B=@n5), B=ScalarParameter() {defaultValue=-0.48, parameterName=Pulse - ??})
  )
  Frac(None=@n6)
  n7 = FunctionInput() {inputName=Result, inputType=FunctionInput_MaterialAttributes}
  n8 = Multiply(A=NamedRerouteUsage(), B=ScalarParameter() {defaultValue=0.4, group=Wing Bath - Pickaxe - Offhand, parameterName=Scene Color - Darkening})
  n9 = NamedRerouteUsage()
  n10 = NamedRerouteUsage()
  n11 = VectorParameter() {group=Wing Bath - Pickaxe - Offhand, parameterName=Flare - Color, defaultValue=<rgba 0.0234 0.4087 0.5 1>}
  n12 = Add(
   A=Multiply(
    A=VectorParameter() {group=Wing Bath - Pickaxe - Offhand, parameterName=Main Color, defaultValue=<rgba 0.1094 0.7043 1 1>}.RGB
    B=Max(A=@n1, B=@n2)
   )
   B=Multiply(A=Max(A=@n9, B=@n10), B=@n11.RGB)
  )
  n13 = LinearInterpolate(
   A=@n12
   B=Multiply(A=@n12, B=ScalarParameter() {defaultValue=0.64, group=Wing Bath - Pickaxe - Offhand, parameterName=Flicker - Min Value})
   Alpha=NamedRerouteUsage()
  )
  n14 = MaterialFunctionCall!(Opacity=Clamp(
    None=NamedRerouteUsage()
    Min=ScalarParameter() {defaultValue=0.65, group=Wing Bath - Pickaxe - Offhand, parameterName=Opacity - Min}
    Max=ScalarParameter() {defaultValue=0.85, group=Wing Bath - Pickaxe - Offhand, parameterName=Opacity - Max}
   )) {materialFunction=/Game/Global/Materials/MaterialGym/Materials/SortedOutline/MF_SortTranslucency}
  n15 = ScalarParameter() {defaultValue=0.3, group=Wing Bath - Pickaxe - Offhand, parameterName=UnderWater_MinOpacity, sliderMax=1}
  n16 = ScalarParameter() {defaultValue=0.8, group=Wing Bath - Pickaxe - Offhand, parameterName=UnderWater_MaxOpacity, sliderMax=1}
  n17 = Subtract(A=PixelDepth().R, B=ComponentMask(None=SceneTexture() {sceneTextureId=PPI_SceneDepth}.Color) {r=true})
  n18 = ScalarParameter() {defaultValue=100, group=Wing Bath - Pickaxe - Offhand, parameterName=UnderWater_DistanceBlend, sliderMax=1024}
  n19 = OneMinus(None=Saturate(None=Divide(A=@n17, B=@n18)))
  n20 = Ceil(None=Saturate(None=@n17))
  n21 = LinearInterpolate(A=@n14.Result, B=Multiply(A=@n14.Result, B=LinearInterpolate(A=@n15, B=@n16, Alpha=@n19) {constA=0.2, constB=0.2}), Alpha=@n20) {desc=above and under water level}
  n22 = Multiply(A=@n14.Result, B=OneMinus(None=@n20))
  n23 = ComponentMask(None=SceneTexture(UVs=ScreenPosition().ViewportUV) {sceneTextureId=PPI_ShadingModelID}.Color) {r=true}
  n24 = Constant() {desc=ID of SingleLayerWater, r=10}
  n25 = Saturate(None=Abs(None=Subtract(A=@n23, B=@n24)))
  FunctionOutput(None=SetMaterialAttributes!(
    MaterialAttributes=GetMaterialAttributes(None=@n7) {attributeGetTypes [#1, #2]}.MaterialAttributes
    EmissiveColor=Add(A=@n8, B=@n13)
    Opacity=LinearInterpolate(A=@n21, B=@n22, Alpha=@n25) {desc=infront or under water}
   ) {attributeSetTypes [#3, #4]}) {outputName=SetMaterialAttributes}
  n26 = Sine(None=Add(A=@n6, B=ScalarParameter() {group=Wing Bath - Pickaxe - Offhand, parameterName=Secondary Pulse - Time Offset}))
  n27 = Step(
   Y=OneMinus(None=@n26)
   X=ScalarParameter() {defaultValue=1.5602, group=Wing Bath - Pickaxe - Offhand, parameterName=Secondary Pulse - Step 2}
  )
  Multiply(
   A=Step(Y=@n26, X=ScalarParameter() {defaultValue=0.0299, group=Wing Bath - Pickaxe - Offhand, parameterName=Secondary Pulse - Step}) {constX=0.95}
   B=@n27
  )
  n28 = OneMinus(None=NamedRerouteUsage())
  n29 = Power(Base=@n28, Exp=ScalarParameter() {defaultValue=8, group=Wing Bath - Pickaxe - Offhand, parameterName=Concentric Rings - Fade Dis})
  Multiply(A=@n29, B=ScalarParameter() {defaultValue=1, parameterName=Main Pulse - Max Intensity}) {constB=5}
  n30 = MaterialFunctionCall(
   UVs=BumpOffset(Height=ScalarParameter() {defaultValue=-1, group=Wing Bath - Pickaxe - Offhand, parameterName=Center Glow - Depth})
   Texture Scale=ScalarParameter() {defaultValue=0.2, group=Wing Bath - Pickaxe - Offhand, parameterName=Center Glow - Scale}
  ) {materialFunction=&ScaleUVsByCenter}
  n31 = TextureSample(UVs=@n30.UVs) {samplerSource=SSM_Clamp_WorldGroupSettings, texture=/BRCosmetics/Weapons/FORT_Melee/Pickaxe_WingBath/Textures/T_Wing_Bath_Circle}
  n32 = ScalarParameter() {defaultValue=1, group=Wing Bath - Pickaxe - Offhand, parameterName=Center Glow - Exp}
  NamedRerouteDeclaration(None=Multiply(
    A=Power(Base=@n31.RGB, Exp=@n32) {constExponent=1.984}
    B=ScalarParameter() {defaultValue=50, group=Wing Bath - Pickaxe - Offhand, parameterName=Center Glow - Intensity}
   )) {name=Center Glow, nodeColor=<rgba 0.0078 0 1 1>}
  n33 = ScalarParameter() {defaultValue=0.05, group=Wing Bath - Pickaxe - Offhand, parameterName=Inner Streaks - Step}
  n34 = Add(A=@n33, B=ScalarParameter() {defaultValue=5, group=Wing Bath - Pickaxe - Offhand, parameterName=Inner Streaks - Smoothness})
  n35 = OneMinus(None=Power(
    Base=NamedRerouteUsage()
    Exp=ScalarParameter() {defaultValue=0.1115, group=Wing Bath - Pickaxe - Offhand, parameterName=Inner Streaks - Exp}
   ))
  NamedRerouteDeclaration(None=Multiply(
    A=ScalarParameter() {defaultValue=10000, group=Wing Bath - Pickaxe - Offhand, parameterName=Inner Streaks - Intensity}
    B=SmoothStep(Min=@n33, Max=@n34, Value=@n35)
   )) {name=Streaks Inner Intensity, nodeColor=<rgba 0.5 0 1 1>}
  n36 = TextureCoordinate()
  n37 = MaterialFunctionCall(Vector or UVs=Subtract(A=@n36) {constB=0.5}) {materialFunction=&VectorToRadialValue}
  NamedRerouteDeclaration(None=ComponentMask(None=@n37.Radial Coordinates) {g=true}) {name=Ring Coordinates, nodeColor=<rgba 1 0 0.1172 1>}
  n38 = VectorParameter() {group=Wing Bath - Pickaxe - Offhand, parameterName=Outer Ring - Pos, defaultValue=<rgba 0.501 0.504 0 1>}
  n39 = MaterialFunctionCall(
   CenterPosition=@n38.RGB
   Radius=ScalarParameter() {defaultValue=0.6299, group=Wing Bath - Pickaxe - Offhand, parameterName=Outer Ring - Radius}
   Density=ScalarParameter() {defaultValue=2.35, group=Wing Bath - Pickaxe - Offhand, parameterName=Outer Ring - Density}
  ) {materialFunction=&RadialGradientExponential}
  NamedRerouteDeclaration(None=Step(
    Y=ScalarParameter() {defaultValue=1.004, group=Wing Bath - Pickaxe - Offhand, parameterName=Outer Ring - Mask}
    X=@n39.RadialGradientExponential
   ) {constX=0.1, constY=0.1}) {name=Shield Border, nodeColor=<rgba 1 0.1875 0 1>}
  n40 = ScalarParameter() {defaultValue=0.5, group=Wing Bath - Pickaxe - Offhand, parameterName=Pulse Fade - Min}
  n41 = ScalarParameter() {defaultValue=1, group=Wing Bath - Pickaxe - Offhand, parameterName=Pulse Fade - Max}
  n42 = Power(Base=NamedRerouteUsage(), Exp=ScalarParameter() {defaultValue=1, group=Wing Bath - Pickaxe - Offhand, parameterName=Pulse Fade - Power})
  NamedRerouteDeclaration(None=SmoothStep(Min=@n40, Max=Add(A=@n40, B=@n41), Value=OneMinus(None=@n42))) {name=Pulse Fade, nodeColor=<rgba 0 1 0.3672 1>}
  NamedRerouteDeclaration(None=Power(Base=@n28, Exp=ScalarParameter() {defaultValue=25, group=Wing Bath - Pickaxe - Offhand, parameterName=Inner Ring - Bias Exp})) {name=Inner Ring - Bias, nodeColor=<rgba 0 0.0625 1 1>}
  NamedRerouteDeclaration(None=ComponentMask(None=@n37.Radial Coordinates) {r=true}) {name=Streak Coordinates, nodeColor=<rgba 0 0.4375 1 1>}
  NamedRerouteDeclaration(None=@n29) {name=Concentric Rings Fade, nodeColor=<rgba 0.7344 1 0 1>}
  n43 = NamedRerouteUsage()
  n44 = Time()
  n45 = MaterialFunctionCall!(
   Sine Phase=Multiply(A=@n44, B=ScalarParameter() {defaultValue=3, group=Wing Bath - Pickaxe - Offhand, parameterName=Secondary Jitter - Speed})
   Value 1=ScalarParameter() {group=Wing Bath - Pickaxe - Offhand, parameterName=Secondary Streak - Jitter Min}
   Value 2=ScalarParameter() {defaultValue=0.2, group=Wing Bath - Pickaxe - Offhand, parameterName="Secondary  Streak - Jitter Max"}
  ) {materialFunction=/Engine/Functions/Engine_MaterialFunctions02/Utility/Sine_Remapped}
  n46 = Multiply(A=@n44, B=ScalarParameter() {defaultValue=4, group=Wing Bath - Pickaxe - Offhand, parameterName=Streak - Jitter - Speed})
  n47 = ScalarParameter() {group=Wing Bath - Pickaxe - Offhand, parameterName=Streak - Jitter - Min}
  n48 = ScalarParameter() {defaultValue=0.05, group=Wing Bath - Pickaxe - Offhand, parameterName=Streak - Jitter - Max}
  n49 = NamedRerouteDeclaration(None=Frac(None=MaterialFunctionCall(Sine Phase=@n46, Value 1=@n47, Value 2=@n48) {materialFunction=&Sine_Remapped}.Remapped Sine)) {name=Jitter, nodeColor=<rgba 1 0.2344 0 1>}
  n50 = Add(A=OneMinus(None=Frac(None=@n45.Remapped Sine)), B=@n49) {constB=0.3}
  n51 = ScalarParameter() {defaultValue=8, group=Wing Bath - Pickaxe - Offhand, parameterName=Streak - Amounts}
  n52 = Frac(None=Multiply(A=Add(A=@n43, B=@n50), B=@n51))
  n53 = OneMinus(None=@n52)
  n54 = NamedRerouteUsage()
  n55 = Frac(None=Multiply(A=@n51, B=Add(A=@n49, B=@n54)))
  n56 = OneMinus(None=@n55)
  n57 = Multiply(
   A=Min(A=Multiply(A=@n53, B=@n52), B=Multiply(A=@n56, B=@n55))
   B=ScalarParameter() {defaultValue=5, group=Wing Bath - Pickaxe - Offhand, parameterName=Streak - Thickness}
  )
  n58 = Multiply(
   A=MaterialFunctionCall(In=OneMinus(None=@n57), Contrast=ScalarParameter() {defaultValue=0.1, group=Wing Bath - Pickaxe - Offhand, parameterName=Streak - Contrast}) {materialFunction=&CheapContrast}.Result
   B=ScalarParameter() {defaultValue=0.01, group=Wing Bath - Pickaxe - Offhand, parameterName=Streak - Intensity}
  )
  n59 = NamedRerouteUsage()
  NamedRerouteDeclaration(None=LinearInterpolate(A=@n58, B=Add(A=@n58, B=@n59), Alpha=Saturate(None=@n58))) {name=Streaks, nodeColor=<rgba 0.5234 1 0 1>}
  n60 = Multiply(A=Time(), B=ScalarParameter() {defaultValue=11.3943, parameterName=Flicker - Speed})
  n61 = ScalarParameter() {parameterName=Flicker - Min}
  NamedRerouteDeclaration(None=Frac(None=MaterialFunctionCall(Sine Phase=@n60, Value 1=@n61) {materialFunction=&Sine_Remapped}.Remapped Sine)) {name=Flicker, nodeColor=<rgba 0.9688 1 0 1>}
  n62 = ScalarParameter() {defaultValue=0.0003, group=Wing Bath - Pickaxe - Offhand, parameterName=Outer Ring - Thickness}
  n63 = Sine(None=@n39.RadialGradientExponential)
  n64 = OneMinus(None=@n63)
  n65 = Saturate(None=Power(
    Base=Divide(A=@n62, B=@n64) {constA=0.0015}
    Exp=ScalarParameter() {defaultValue=0.7, group=Wing Bath - Pickaxe - Offhand, parameterName=Outer Ring - Exp}
   ))
  n66 = ScalarParameter() {defaultValue=0.9009, group=Wing Bath - Pickaxe - Offhand, parameterName=Outer Ring - Fade Step}
  n67 = Add(A=@n66, B=ScalarParameter() {defaultValue=1, group=Wing Bath - Pickaxe - Offhand, parameterName=Outer Ring - Fade Smoothness})
  n68 = ScalarParameter() {defaultValue=0.3799, group=Wing Bath - Pickaxe - Offhand, parameterName=Outer Ring - Second Fade - Step}
  n69 = SmoothStep(
   Min=@n68
   Max=Add(A=@n68, B=ScalarParameter() {defaultValue=4, group=Wing Bath - Pickaxe - Offhand, parameterName="Outer Ring -  Second Fade - Smoothness"})
   Value=@n63
  ) {constMax=4}
  n70 = ScalarParameter() {defaultValue=0.2, group=Wing Bath - Pickaxe - Offhand, parameterName=Outer Ring - Second Fade - Intensity}
  n71 = Multiply(
   A=Max(A=SmoothStep(Min=@n66, Max=@n67, Value=@n63) {constMax=4}, B=Multiply(A=@n69, B=@n70))
   B=ScalarParameter() {defaultValue=1, group=Wing Bath - Pickaxe - Offhand, parameterName=Outer Ring - Intensity}
  )
  n72 = Multiply(
   A=Clamp(
    None=Add(A=@n65, B=@n71)
    Max=ScalarParameter() {defaultValue=0.5, group=Wing Bath - Pickaxe - Offhand, parameterName=Outer Ring - Max Brightness}
   )
   B=ScalarParameter() {defaultValue=1, group=Wing Bath - Pickaxe - Offhand, parameterName=Outer Ring - Global Opacity}
  )
  NamedRerouteDeclaration(None=Max(A=NamedRerouteUsage(), B=NamedRerouteDeclaration(None=@n72) {name=Outer Ring, nodeColor=<rgba 0 0.6484 1 1>})) {name=Pulses + Outer Ring, nodeColor=<rgba 1 0 0.8437 1>}
  n73 = Time()
  n74 = ScalarParameter() {defaultValue=0.2, parameterName=Pulse - Sine Speed}
  NamedRerouteDeclaration(None=MaterialFunctionCall(Sine Phase=Multiply(A=@n73, B=@n74), Value 1=ScalarParameter() {defaultValue=0.2, parameterName=Pulse - Min}) {materialFunction=&Sine_Remapped}.Remapped Sine) {name=Sine Pulse, nodeColor=<rgba 0 1 0.9531 1>}
  n75 = Multiply(A=Time(), B=ScalarParameter() {defaultValue=-0.2, group=Wing Bath - Pickaxe - Offhand, parameterName=Variation Mask - Speed})
  n76 = NamedRerouteUsage()
  n77 = Multiply(
   A=Add(A=NamedRerouteUsage(), B=Multiply(A=@n75, B=@n76))
   B=ScalarParameter() {defaultValue=20, group=Wing Bath - Pickaxe - Offhand, parameterName=Variation Mask - Tiling}
  )
  NamedRerouteDeclaration(None=Saturate(None=Sine(None=@n77))) {name=Variation Mask, nodeColor=<rgba 0 1 0.6719 1>}
  n78 = NamedRerouteUsage()
  n79 = Multiply(A=Time(), B=ScalarParameter() {defaultValue=-0.1, group=Wing Bath - Pickaxe - Offhand, parameterName=Turbulence - Noise Speed})
  n80 = NamedRerouteUsage()
  n81 = Multiply(
   A=AppendVector(A=Multiply(A=@n78) {constB=2}, B=Add(A=@n79, B=@n80))
   B=ScalarParameter() {defaultValue=2, group=Wing Bath - Pickaxe - Offhand, parameterName=Turbulence - Tiling}
  )
  NamedRerouteDeclaration(None=TextureSample!(UVs=@n81, Tex, "DDX(UVs)"=NamedRerouteUsage(), "DDY(UVs)"=NamedRerouteUsage(), Apply View MipBias) {mipValueMode=TMVM_Derivative, texture=/BRCosmetics/Animation/Game/MainPlayer/Emotes/VelvetDeskCam/FX/Swirly_Wisps_Vertical_Dissolve}.RGB) {name=Turbulence, nodeColor=<rgba 1 0.3281 0 1>}
  n82 = Multiply(A=Time(), B=ScalarParameter() {defaultValue=-0.15, group=Wing Bath - Pickaxe - Offhand, parameterName=Pulse - Speed})
  n83 = NamedRerouteDeclaration(None=ScalarParameter() {defaultValue=1, parameterName=Global SPeed}) {name=Global Speed, nodeColor=<rgba 0.875 1 0 1>}
  n84 = Add(
   A=Add(A=NamedRerouteUsage(), B=Multiply(A=@n82, B=@n83))
   B=ScalarParameter() {defaultValue=0.414, group=Wing Bath - Pickaxe - Offhand, parameterName=Pulse - Offset}
  )
  n85 = Multiply(A=OneMinus(None=@n84), B=ScalarParameter() {defaultValue=6, group=Wing Bath - Pickaxe - Offhand, parameterName=Pulse - Frequency})
  n86 = Frac(None=Multiply(A=@n84, B=ScalarParameter() {defaultValue=31, group=Wing Bath - Pickaxe - Offhand, parameterName=Pulse - Tiling}))
  n87 = Multiply(
   A=Multiply(A=@n86, B=OneMinus(None=@n86))
   B=ScalarParameter() {defaultValue=3, group=Wing Bath - Pickaxe - Offhand, parameterName=Pulse - Thickness}
  )
  n88 = Multiply(
   A=Step(Y=ScalarParameter() {defaultValue=0.6, group=Wing Bath - Pickaxe - Offhand, parameterName=Pulse - Step}, X=Frac(None=@n85)) {constX=0.95}
   B=Multiply(A=OneMinus(None=@n87), B=ScalarParameter() {defaultValue=1, group=Wing Bath - Pickaxe - Offhand, parameterName=Pulse - Intensity})
  )
  n89 = NamedRerouteUsage()
  n90 = ScalarParameter() {defaultValue=0.5, parameterName=Band Width}
  n91 = ScalarParameter() {parameterName="Sharpness (S)"}
  n92 = ScalarParameter() {parameterName=Compare}
  n93 = Multiply(A=@n3, B=ScalarParameter() {defaultValue=19, parameterName=Bands})
  n94 = Multiply(A=Multiply(A=@n4, B=ScalarParameter() {defaultValue=-5, group=Wing Bath - Pickaxe - Offhand, parameterName=Secondary Pulse - Speed}), B=@n5)
  n95 = MaterialFunctionCall!(Width=@n90, Sharpness=@n91, Offset, Direction Switch, Compare=@n92, Input Coordinates=Add(A=@n93, B=@n94)) {materialFunction=/Engine/Functions/Engine_MaterialFunctions03/Procedurals/GeneratedBand}
  n96 = MaterialFunctionCall(Width=@n90, Sharpness=@n91, Compare=@n92, Input Coordinates=OneMinus(None=Add(A=@n93, B=@n94))) {materialFunction=&GeneratedBand}
  n97 = Multiply(
   A=Multiply(
    A=Max(A=@n95.Result, B=@n96.Result)
    B=ScalarParameter() {defaultValue=5, group=Wing Bath - Pickaxe - Offhand, parameterName="Second Pulse  - Intensity"}
   )
   B=@n27
  )
  n98 = Add(A=Time()) {constB=0.75}
  n99 = ScalarParameter() {defaultValue=0.25, parameterName=Main Pulse - Sine Speed}
  n100 = Multiply(
   A=LinearInterpolate(A=Multiply(A=@n97) {constB=-0.3}, B=Multiply(A=@n97) {constB=4}, Alpha=NamedRerouteUsage()) {constB=4}
   B=MaterialFunctionCall(
    Sine Phase=Multiply(A=@n98, B=@n99)
    Value 1=ScalarParameter() {defaultValue=0.5, parameterName=Main Pulse - Sine Min}
    Value 2=ScalarParameter() {defaultValue=0.5, parameterName=Main Pulse - Sine Max}
   ) {materialFunction=&Sine_Remapped}.Remapped Sine
  )
  n101 = NamedRerouteUsage()
  n102 = NamedRerouteUsage()
  n103 = Multiply(A=Multiply(A=Time(), B=ScalarParameter() {defaultValue=-0.15, parameterName=Third Pulse - Speed}), B=@n102)
  n104 = Frac(None=Multiply(A=Add(A=@n101, B=@n103), B=ScalarParameter() {defaultValue=20, parameterName=Third Pulse - Tiling}))
  n105 = OneMinus(None=@n104)
  n106 = OneMinus(None=Multiply(A=Multiply(A=@n104, B=@n105), B=ScalarParameter() {defaultValue=3, parameterName=Third Pulse - Line Thickness}))
  n107 = ScalarParameter() {defaultValue=3, parameterName="Third Pulse  - Intensity"}
  n108 = ScalarParameter() {defaultValue=0.088, parameterName=Third Pulse - Step Amount}
  n109 = SmoothStep(Min=@n108, Max=Add(A=@n108, B=ScalarParameter() {defaultValue=0.05, parameterName=Third Pulse - Smoothness}), Value=NamedRerouteUsage())
  n110 = ScalarParameter() {defaultValue=-0.16, parameterName=Border Ring - Speed}
  n111 = Time()
  n112 = Add(A=Multiply(A=Multiply(A=@n110, B=@n111), B=@n102), B=NamedRerouteUsage())
  n113 = ScalarParameter() {defaultValue=13, parameterName=Border Ring - Tiling}
  n114 = ScalarParameter() {defaultValue=0.2638, parameterName=Border Ring - Time Offset}
  n115 = Frac(None=Add(A=Multiply(A=@n112, B=@n113), B=@n114))
  n116 = OneMinus(None=@n115)
  n117 = OneMinus(None=Multiply(A=Multiply(A=@n115, B=@n116), B=ScalarParameter() {defaultValue=15, parameterName=Border Ring - Line Thickness}))
  n118 = ScalarParameter() {defaultValue=3, parameterName=Border Ring - Intensity}
  n119 = ScalarParameter() {defaultValue=0.4, parameterName=Border Ring - Step Amount}
  n120 = Sine(None=Multiply(A=Add(A=@n112, B=@n114)) {constB=4})
  n121 = SmoothStep(Min=@n119, Max=Add(A=@n119, B=ScalarParameter() {defaultValue=0.1, parameterName=Border Ring - Step Smoothness}), Value=@n120)
  n122 = Max(A=Multiply(A=Multiply(A=@n106, B=@n107), B=OneMinus(None=@n109)), B=Multiply(A=Multiply(A=@n117, B=@n118), B=@n121))
  n123 = Subtract(A=Max(A=Multiply(A=@n88, B=@n89), B=Max(A=@n100, B=@n122)), B=Multiply(A=NamedRerouteUsage()) {constB=0})
  n124 = LinearInterpolate(
   A=Multiply(A=ScalarParameter() {defaultValue=0.2, parameterName=Pulse - Min}, B=@n123)
   B=Multiply(A=@n123, B=ScalarParameter() {defaultValue=0.8, parameterName=Pulse - Max})
   Alpha=Saturate(None=NamedRerouteUsage())
  ) {constB=0}
  n125 = ScalarParameter() {defaultValue=5, parameterName=Inner Ring - Bias}
  NamedRerouteDeclaration(None=LinearInterpolate(A=@n124, B=Multiply(A=@n124, B=@n125), Alpha=NamedRerouteUsage())) {name=Pulse, nodeColor=<rgba 1 0.1641 0 1>}
  NamedRerouteDeclaration(None=@n81) {name=Shield Movement, nodeColor=<rgba 0.875 0 1 1>}
  NamedRerouteDeclaration(None=DDX(Value=@n36)) {name=DDX, nodeColor=<rgba 1 0.3047 0 1>}
  NamedRerouteDeclaration(None=DDY(Value=@n36)) {name=DDY, nodeColor=<rgba 0 0.0156 1 1>}
  n126 = Constant()
  n127 = TextureSample(UVs=NamedRerouteUsage(), "DDX(UVs)"=NamedRerouteUsage(), "DDY(UVs)"=NamedRerouteUsage()) {
   mipValueMode=TMVM_Derivative
   samplerType=SAMPLERTYPE_Normal
   texture=&/T_Waterfall_NormalA
  }
  n128 = ScalarParameter() {group=Wing Bath - Pickaxe - Offhand, parameterName=Distortion - Fade Intensity Distance}
  n129 = Multiply(
   A=@n127.RGB
   B=LinearInterpolate(
    A=ScalarParameter() {defaultValue=0.005, group=Wing Bath - Pickaxe - Offhand, parameterName=Distortion - Intensity Min}
    B=ScalarParameter() {group=Wing Bath - Pickaxe - Offhand, parameterName=Distortion - Intensity Max}
    Alpha=MaterialFunctionCall(Fade Offset=@n128) {materialFunction=&CameraDepthFade}.Result
   )
  )
  n130 = SceneColor!(UVs=Add(A=ScreenPosition().ViewportUV, B=ComponentMask(None=@n129) {g=true, r=true}))
  NamedRerouteDeclaration(None=MaterialFunctionCall!(Low=@n126, High=@n130.RGB, Epic=@n130.RGB, --------=@n130.RGB, Mobile=@n126, Medium=@n126) {materialFunction=/FortniteBaseMaterials/Functions/MF_QualitySwitch_Vector}.Result) {name=Distorted Scene Color, nodeColor=<rgba 1 0 0.9844 1>}
  n131 = TextureSample!(
   UVs=MaterialFunctionCall(UVs=@n30.UVs, "Rotation Angle (0-1)"=NamedRerouteUsage()) {materialFunction=&CustomRotator}.Rotated Values
   Tex
   Apply View MipBias
  ) {gatherMode=TGM_Red, samplerSource=SSM_Clamp_WorldGroupSettings, texture=/Game/Effects/Fort_Effects/Textures/Flares/T_Flare_Streaks}
  n132 = ScalarParameter() {defaultValue=0.85, group=Wing Bath - Pickaxe - Offhand, parameterName=Flare Glow - Exp}
  NamedRerouteDeclaration(None=Multiply(
    A=Power(Base=@n131.R, Exp=@n132) {constExponent=1.984}
    B=ScalarParameter() {defaultValue=50, group=Wing Bath - Pickaxe - Offhand, parameterName=Flare Glow - Intensity}
   )) {name=Flare, nodeColor=<rgba 0 1 0.5781 1>}
  OneMinus(None=@n120)
  OneMinus(None=@n121)
  OneMinus(None=@n93)
  ScalarParameter() {defaultValue=15, parameterName=Band Amount}
  Sine(None=@n60)
 ]
}
