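# Parent material for the WingBath pickaxe offhand cosmetic, an unlit translucent surface for skeletal meshes.
# It uses material attributes, passing the shared 2023 character PreFX result through the WingBath offhand function and into the shared PostFX function.
# PreFX also feeds PostFX's auxiliary inputs directly, bypassing the WingBath function.
# Depth testing is disabled, while translucent custom depth writes and pixel animation are flagged on.
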Material /BRCosmetics/Weapons/FORT_Melee/Pickaxe_WingBath/Materials/M_Pickaxe_WingBath_Offhand_Parent
graph {
 properties {
  allowTranslucentCustomDepthWrites=true
  bDisableDepthTest=true
  bHasPixelAnimation=true
  bUseMaterialAttributes=true
  bUsedWithMeshDeformer=true
  bUsedWithMorphTargets=true
  bUsedWithSkeletalMesh=true
  blendMode=BLEND_Translucent
  pixelDepthOffsetMode=PDOM_Legacy
  shadingModel=MSM_Unlit
  translucencyLightingMode=TLM_SurfacePerPixelLighting
  usedShadingModels=Unlit
 }
 nodes [
  n1 = MaterialFunctionCall() {materialFunction=&MF_CharacterParent_2023_PreFX}
  MaterialFunctionCall(
   Material=MaterialFunctionCall!(Result=@n1.Result) {materialFunction=&/MF_Pickaxe_WingBath_Offhand_Parent}.SetMaterialAttributes
   M=@n1.M
   Pre Skinned Local Position=@n1.M
   Distance Blend=@n1.M
   Fresnel=@n1.M
  ) {materialFunction=&MF_CharacterParent_2023_PostFX}
 ]
}
